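; Fortnite weapon Blueprint for the Annoyed Mop, which uses main-hand and offhand Niagara components in place of weapon meshes.
; On equip it attaches both effects to the pawn's hand sockets, fires an equip gameplay cue and binds water enter/exit events.
; In water, a looping timer sets an effect's Height parameter from immersion depth, choosing component and offset by crouch state.
; Unequip fires a cue and unbinds water events; swing-trail functions are empty.
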(event FortWeapon|EventOnEquippedAndAttached
  (bind _idle_niagara_offhand (Variables|Default|GetIdle%20Niagara%20Offhand))
  (bind _idlefx_niagara (Variables|Default|GetIdleFX%20Niagara))
  ;[0.7147 1 0.85] On Equip
  (|CallParentFunction)
  (Development|PrintString "EQUIPPED" (Variables|Default|GetbDebug) (Variables|Default|GetbDebug))
  (bind _asfort_pawn (Utilities|Casting|CastToFortPawn (Game|GetInstigator)))
  (Utilities|IsValid (Variables|/Script/Engine.Character|GetMesh _asfort_pawn)
    (:Is%20Valid
      _)
    (:Is%20Not%20Valid
      (Utilities|FlowControl|Delay 0.1)))
  (Utilities|FlowControl|Sequence
    (:then_0
      ;;[1 1 1] Fix weapon position/rotation on equip - Since we don't have actual meshes for this weapon, use the niagara components directly
      (Transformation|AttachComponentToComponent _idlefx_niagara (Variables|/Script/Engine.Character|GetMesh _asfort_pawn) (Variables|WeaponPositionInfo|GetMainhandSocketName) "SnapToTarget" "SnapToTarget" "SnapToTarget" true)
      (Components|Activation|Activate _idlefx_niagara true)
      (Transformation|AttachComponentToComponent _idle_niagara_offhand (Variables|/Script/Engine.Character|GetMesh _asfort_pawn) (Variables|WeaponPositionInfo|GetOffhandSocketName) "SnapToTarget" "SnapToTarget" "SnapToTarget" true)
      (Components|Activation|Activate _idle_niagara_offhand true)
      (Transformation|SetRelativeLocationAndRotation (Variables|Default|GetIdleFX%20Niagara) (Variables|WeaponPositionInfo|GetWeaponRelativeOffset_Main) (Variables|WeaponPositionInfo|GetWeaponRelativeRotation_Main) false true)
      (Transformation|SetRelativeLocationAndRotation (Variables|Default|GetIdle%20Niagara%20Offhand) (Variables|WeaponPositionInfo|GetWeaponRelativeOffset_Offhand) (Variables|WeaponPositionInfo|GetWeaponRelativeRotation_Offhand) false true)
      (Utilities|FlowControl|DoOnce false
        (:Completed
          (GameplayCue|ExecuteGameplayCue (Game|GetInstigator) (Variables|Default|GetOnEquipGameplayCue)))))
    (:then_1
      ;;[1 1 1] Set player skel mesh and offhand parameters for Niagara
      (Niagara|SetNiagaraSkeletalMeshComponent 0 "SkeletalMesh" (Variables|/Script/Engine.Character|GetMesh _asfort_pawn))
      (Niagara|SetNiagaraSkeletalMeshComponent 0 "SkeletalMesh" (Variables|/Script/Engine.Character|GetMesh _asfort_pawn))
      ;;[1 0.2892 0.3456] Enter and Exit Water (VFX)  :TODO Need to get "when squatting" and adjust height
      ;;
      (bind _asfort_player_pawn (Utilities|Casting|CastToFortPlayerPawn _asfort_pawn))
      (EventDispatchers|BindEventtoOnPawnExitWater _asfort_player_pawn (EventDispatchers|CreateEvent :!event Custom|Exit%20Water))
      (EventDispatchers|BindEventtoOnPawnEnterWater _asfort_player_pawn (EventDispatchers|CreateEvent :!event Custom|Enter%20Water))
      (if (> (- (WaterBody|GetWaterImmersionDepth _asfort_player_pawn) 4e+01) 0.1)
        (CallFunction|EnterWater _asfort_player_pawn)
        (else
          (CallFunction|ExitWater _asfort_player_pawn))))))

(event Custom|Enter%20Water
  (params (FortPlayerPawn :type Fort%20Player%20Pawn%20Object%20Reference))
  ;[0.7147 1 0.85] On Equip
  ;;[1 0.2892 0.3456] Enter and Exit Water (VFX)  :TODO Need to get "when squatting" and adjust height
  ;;
  (bind _returnvalue (Utilities|Time|SetTimerbyEvent (delegate Custom|Loop) 0.1 true))
  (Variables|Default|SetWaterLoopTimerHandle _returnvalue))

(event Custom|Loop
  ;[0.7147 1 0.85] On Equip
  ;;[1 0.2892 0.3456] Enter and Exit Water (VFX)  :TODO Need to get "when squatting" and adjust height
  ;;
  (if (AI|Components|NavMovement|IsCrouching (Variables|/Script/Engine.Character|GetCharacterMovement (stale FortWeapon|EventOnEquippedAndAttached _asfort_player_pawn)))
    (Effects|Components|ParticleSystem|SetFloatParameter (Variables|Default|GetIdleFX%20Niagara) "Height" (- 1.0 (Math|Float|Clamp(Float) (- (WaterBody|GetWaterImmersionDepth (stale Custom|Enter%20Water FortPlayerPawn)) 3e+01))))
    (else
      (Effects|Components|ParticleSystem|SetFloatParameter (Variables|Default|GetIdle%20Niagara%20Offhand) "Height" (- 1.0 (Math|Float|Clamp(Float) (- (WaterBody|GetWaterImmersionDepth (stale Custom|Enter%20Water FortPlayerPawn)) 75.0)))))))

(event Custom|Exit%20Water
  (params (FortPlayerPawn :type Fort%20Player%20Pawn%20Object%20Reference) (bPawnIsOutsideOfAllWaterBodies :type Boolean))
  ;[0.7147 1 0.85] On Equip
  ;;[1 0.2892 0.3456] Enter and Exit Water (VFX)  :TODO Need to get "when squatting" and adjust height
  ;;
  (Utilities|Time|ClearandInvalidateTimerbyHandle (Variables|Default|GetWaterLoopTimerHandle))
  (Effects|Components|ParticleSystem|SetFloatParameter (Variables|Default|GetIdle%20Niagara%20Offhand) "Height" 1.0))

(event FortWeapon|EventOnLocalUnEquip
  (bind _returnvalue (Game|GetInstigator))
  ;[1 0.7832 0.7128] On Unequip / Detached
  (|CallParentFunction)
  (Components|Activation|Deactivate (Variables|Default|GetIdleFX%20Niagara))
  (Transformation|DetachFromComponent (Variables|Default|GetIdle%20Niagara%20Offhand) "KeepRelative" "KeepRelative" "KeepRelative" true)
  (GameplayCue|ExecuteGameplayCue _returnvalue (Variables|Default|GetOnUnequipGameplayCue))
  ;[1 1 1] Unbind water events
  (bind _asfort_player_pawn (Utilities|Casting|CastToFortPlayerPawn _returnvalue))
  (EventDispatchers|UnbindEventfromOnPawnEnterWater _asfort_player_pawn (EventDispatchers|CreateEvent :!event Custom|Enter%20Water))
  (EventDispatchers|UnbindEventfromOnPawnExitWater _asfort_player_pawn (EventDispatchers|CreateEvent :!event Custom|Exit%20Water))
  (CallFunction|ExitWater 0))

(event FortWeapon|EventOnUnEquippedAndDetached
  (bind _returnvalue (Game|GetInstigator))
  ;[1 0.7832 0.7128] On Unequip / Detached
  (|CallParentFunction)
  (Components|Activation|Deactivate (Variables|Default|GetIdleFX%20Niagara))
  (Transformation|DetachFromComponent (Variables|Default|GetIdle%20Niagara%20Offhand) "KeepRelative" "KeepRelative" "KeepRelative" true)
  (GameplayCue|ExecuteGameplayCue _returnvalue (Variables|Default|GetOnUnequipGameplayCue))
  ;[1 1 1] Unbind water events
  (bind _asfort_player_pawn (Utilities|Casting|CastToFortPlayerPawn _returnvalue))
  (EventDispatchers|UnbindEventfromOnPawnEnterWater _asfort_player_pawn (EventDispatchers|CreateEvent :!event Custom|Enter%20Water))
  (EventDispatchers|UnbindEventfromOnPawnExitWater _asfort_player_pawn (EventDispatchers|CreateEvent :!event Custom|Exit%20Water))
  (CallFunction|ExitWater 0))

(fn BindSwingandAnimTrailEvents)

(fn UnbindSwingAndAnimTrailEvents)

(fn ConstructionScript
  (|CallParentFunction))
