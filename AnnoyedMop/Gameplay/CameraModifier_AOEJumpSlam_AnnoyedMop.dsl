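; Camera modifier for the AnnoyedMop AOE jump slam that pulls the view back and lags it vertically as the player moves from an anchor point.
; TriggerPullBack and OnLanded set flags, consumed next update to record anchor and landing times.
; Backward offset follows clamped travel along the view and a curve, switching on landing to a montage-length curve; vertical compensation fades over time.
; DisableAndRemove waits out the blend before removing the modifier.
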(event Custom|DisableAndRemove
  (if (not (Variables|Default|GetAlreadyDisabled))
    (Variables|Default|SetAlreadyDisabled true)
    (CameraModifier|DisableModifier false)
    (Utilities|FlowControl|Delay (Variables|CameraModifier|GetAlphaOutTime))
    (Variables|Default|SetRemoved true)
    (bind _cameraowner (Variables|CameraModifier|GetCameraOwner)
      (:then
        (Utilities|IsValid (Game|Player|GetOwningPlayerController _cameraowner)
          (:Is%20Valid
            (Game|Player|RemoveCameraModifier (Variables|/Script/Engine.PlayerController|GetPlayerCameraManager (Game|Player|GetOwningPlayerController _cameraowner)) self))))
      (:else))))

(fn BlueprintModifyCamera
  (params (DeltaTime :type Float%20%28single-precision%29) (ViewLocation :type Vector :default "0, 0, 0") (ViewRotation :type Rotator :default "0, 0, 0") (FOV :type Float%20%28single-precision%29))
  (bind (_returnvalue_x _returnvalue_y _returnvalue_z) (Math|Vector|BreakVector (- (Variables|Default|GetAnchoredLocation) (Transformation|GetActorLocation (Variables|Default|GetPlayerPawn_Athena)))))
  (Utilities|FlowControl|Sequence
    (:then_0
      (Utilities|IsValid (Variables|Default|GetPlayerPawn_Athena)
        (:Is%20Not%20Valid
          (bind _asplayer_pawn_athena (Utilities|Casting|CastToPlayerPawn_Athena (Game|Player|GetPlayerPawn (Utilities|Casting|CastToFortPlayerController (Game|Player|GetOwningPlayerController (Variables|CameraModifier|GetCameraOwner))))))
          (Variables|Default|SetPlayerPawn_Athena _asplayer_pawn_athena)
          (Variables|Default|SetTriggerPullBackNextFrame true)
          (Variables|Default|SetLanded false))))
    (:then_1
      (if (Variables|Default|GetTriggerPullBackNextFrame)
        (Variables|Default|SetTriggerPullBackNextFrame false)
        (Variables|Default|SetAnchoredLocation (Transformation|GetActorLocation (Variables|Default|GetPlayerPawn_Athena)))
        (Variables|Default|SetLastAnchoredTime (Utilities|Time|GetGameTimeinSeconds))))
    (:then_2
      (if (Variables|Default|GetTriggerLandedNextFrame)
        (Variables|Default|SetTriggerLandedNextFrame false)
        (Variables|Default|SetLastLandedTime (Utilities|Time|GetGameTimeinSeconds))
        (Variables|Default|SetLanded true)
        (Variables|Default|SetLandedForward (Variables|Default|GetForwardFromAnchor))))
    (:then_3
      (Variables|Default|SetForwardFromAnchor (* (Math|Float|Clamp(Float) (Math|Vector|DotProduct (Math|Vector|GetForwardVector (Variables|GetViewRotation)) (- (Transformation|GetActorLocation (Variables|Default|GetPlayerPawn_Athena)) (Variables|Default|GetAnchoredLocation))) (Utilities|RemapFloat (- (Utilities|Time|GetGameTimeinSeconds) (Variables|Default|GetLastAnchoredTime)) "(X=0.000000,Y=1.200000)" "(X=0.000000,Y=700.000000)" true) 3e+03) (Math|Curves|GetFloatValue (Variables|Default|GetForwardPrecentCurve) (Utilities|NormalizeFloat (- (Utilities|Time|GetGameTimeinSeconds) (Variables|Default|GetLastAnchoredTime)) 0 3.6 true))))
      (Variables|Default|SetForwardFromLanded (Utilities|RemapFloat (Math|Curves|GetFloatValue (Variables|Default|GetLandedForwardCurve) (Utilities|NormalizeFloat (- (Utilities|Time|GetGameTimeinSeconds) (Variables|Default|GetLastLandedTime)) 0 (Animation|GetPlayLength (Variables|Default|GetLandingMontage)) true)) "(X=0.000000,Y=1.000000)" (Math|Vector2D|MakeVector2D 0.0 (Variables|Default|GetLandedForward)) false))
      (Variables|Default|SetCurrentUp (* _returnvalue_z (Utilities|RemapFloat (- (Utilities|Time|GetGameTimeinSeconds) (Variables|Default|GetLastAnchoredTime)) (0.0 2.0) (0.825 0.0) true))))
    (:then_4
      (return (Utilities|Operators|Add (Variables|GetViewLocation) (Utilities|Operators|Multiply (Math|Vector|GetForwardVector (Variables|GetViewRotation)) (Math|Float|SelectFloat (Variables|Default|GetForwardFromLanded) (Variables|Default|GetForwardFromAnchor) (Variables|Default|GetLanded)) -1.0 (Utilities|GetCurvedAlpha)) (Utilities|Operators|Multiply (Math|Vector|GetUpVector (Variables|GetViewRotation)) (Variables|Default|GetCurrentUp) (Utilities|GetCurvedAlpha))) (Variables|GetViewRotation) (Variables|GetFOV)))))

(fn OnLanded
  (Variables|Default|SetTriggerLandedNextFrame true))

(fn SetStartingLocation
  (params (StartingLocation :type Vector))
  (Variables|Default|SetStartingLocation StartingLocation))

(fn TriggerPullBack
  (params (CurrentPullBackTime :type Float%20%28double-precision%29) (CurrentPullBackPercentage :type Float%20%28double-precision%29))
  (Variables|Default|SetTriggerPullBackNextFrame true)
  (Variables|Default|SetCurrentPullBackTime CurrentPullBackTime)
  (Variables|Default|SetCurrentPullBackPercentage CurrentPullBackPercentage))
