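; Camera modifier for the AnnoyedMop gameplay that pulls the player's view back, shifts it right and widens FOV during a combo.
; TriggerPullBack re-anchors pawn location and current offset next frame, then curve-blends a share of distance travelled into backward offset and raises goal FOV.
; Side offset ramps from combo start time; all offsets scale by curve-remapped modifier alpha.
; DisableAndRemove waits out the alpha fade before removing the modifier.
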(event Custom|DisableAndRemove
  (if (not (Variables|Default|GetAlreadyDisabled))
    (Variables|Default|SetAlreadyDisabled true)
    (CameraModifier|DisableModifier false)
    (Utilities|FlowControl|Delay (Variables|CameraModifier|GetAlphaOutTime))
    (Variables|Default|SetRemoved true)
    (bind _cameraowner (Variables|CameraModifier|GetCameraOwner)
      (:then
        (Utilities|IsValid (Game|Player|GetOwningPlayerController _cameraowner)
          (:Is%20Valid
            (Game|Player|RemoveCameraModifier (Variables|/Script/Engine.PlayerController|GetPlayerCameraManager (Game|Player|GetOwningPlayerController _cameraowner)) self))))
      (:else))))

(fn BlueprintModifyCamera
  (params (DeltaTime :type Float%20%28single-precision%29) (ViewLocation :type Vector :default "0, 0, 0") (ViewRotation :type Rotator :default "0, 0, 0") (FOV :type Float%20%28single-precision%29))
  (Utilities|FlowControl|Sequence
    (:then_0
      (Utilities|IsValid (Variables|Default|GetPlayerPawn_Athena)
        (:Is%20Not%20Valid
          (bind _asplayer_pawn_athena (Utilities|Casting|CastToPlayerPawn_Athena (Game|Player|GetPlayerPawn (Utilities|Casting|CastToFortPlayerController (Game|Player|GetOwningPlayerController (Variables|CameraModifier|GetCameraOwner))))))
          (Variables|Default|SetPlayerPawn_Athena _asplayer_pawn_athena))))
    (:then_1
      (if (Variables|Default|GetTriggerPullBackNextFrame)
        (Variables|Default|SetTriggerPullBackNextFrame false)
        (Variables|Default|SetAnchoredLocation (Transformation|GetActorLocation (Variables|Default|GetPlayerPawn_Athena)))
        (Variables|Default|SetAnchored%20Forward (Variables|Default|GetCurrentForward))
        (Variables|Default|SetLastAnchoredTime (Utilities|Time|GetGameTimeinSeconds))))
    (:then_2
      (Variables|Default|SetCurrentForward (+ (* (Math|Vector|VectorLength (- (Variables|Default|GetAnchoredLocation) (Transformation|GetActorLocation (Variables|Default|GetPlayerPawn_Athena)))) (Utilities|UnNormalizeFloat (Math|Curves|GetFloatValue (Variables|Default|GetPullBackCurve) (Utilities|NormalizeFloat (- (Utilities|Time|GetGameTimeinSeconds) (Variables|Default|GetLastAnchoredTime)) 0 (Variables|Default|GetCurrentPullBackTime) true)) (Variables|Default|GetCurrentPullBackPercentage) (Variables|Default|GetCurrent%20Pull%20Back%20Goal%20Percentage) true)) (Variables|Default|GetAnchored%20Forward)))
      (Variables|Default|SetCurrentFOV (Math|Interpolation|FInterpTo (Variables|Default|GetCurrentFOV) (Variables|Default|GetGoalFOV) (Variables|GetDeltaTime) 2.0)))
    (:then_3
      (return (Utilities|Operators|Add (Variables|GetViewLocation) (Utilities|Operators|Multiply (Math|Vector|GetForwardVector (Variables|GetViewRotation)) (Variables|Default|GetCurrentForward) -1.0 (CallFunction|GetCurvedAlpha)) (Utilities|Operators|Multiply (Math|Vector|GetRightVector (Variables|GetViewRotation)) (Utilities|RemapFloat (- (Utilities|Time|GetGameTimeinSeconds) (Variables|Default|GetComboStartedTime)) "(X=0.000000,Y=2.500000)" "(X=0.000000,Y=25.000000)" true) (CallFunction|GetCurvedAlpha))) (Variables|GetViewRotation) (+ (Variables|GetFOV) (* (Variables|Default|GetCurrentFOV) (CallFunction|GetCurvedAlpha)))))))

(fn GetCurvedAlpha
  (return (Math|Curves|GetFloatValue (Variables|Default|GetAlphaCurve) (Variables|CameraModifier|GetAlpha))))

(fn SetStartingLocation
  (params (StartingLocation :type Vector))
  (Variables|Default|SetStartingLocation StartingLocation))

(fn TriggerComboStart
  (Variables|Default|SetComboStartedTime (Utilities|Time|GetGameTimeinSeconds)))

(fn TriggerPullBack
  (params (Pull%20Back%20Time :type Float%20%28double-precision%29) (Start%20Pul%20Back%20Percentage :type Float%20%28double-precision%29) (End%20Pull%20Back%20Percentage :type Float%20%28double-precision%29))
  (Variables|Default|SetTriggerPullBackNextFrame true)
  (Variables|Default|SetCurrentPullBackTime Pull%20Back%20Time)
  (Variables|Default|SetCurrentPullBackPercentage Start%20Pul%20Back%20Percentage)
  (Variables|Default|SetCurrent%20Pull%20Back%20Goal%20Percentage End%20Pull%20Back%20Percentage)
  (Variables|Default|SetGoalFOV (Math|Float|Clamp(Float) (+ (Variables|Default|GetGoalFOV) 3.0) 0.0 2e+01)))
