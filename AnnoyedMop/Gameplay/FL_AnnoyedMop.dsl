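; Blueprint function library of helpers for aiming and spawning Annoyed Mop arc projectiles.
; ThrowArcProjectile spawns the projectile without default physics or gravity, using AdjustSpawnRotation to roll it around the forward axis while travel direction keeps the untilted rotation.
; AdjustArcProjectileSpawnDirection separately offsets a forward direction along local right and world up, returning a rotation.
; ThrowArcProjectile's Scale input is unused.
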(fn AdjustArcProjectileSpawnDirection
  (params (BaseRotation :type Rotator) (Direction%20Offset%20Right :type Float%20%28double-precision%29) (Direction%20Offset%20Up :type Float%20%28double-precision%29))
  (return (Math|Conversions|RotationFromXVector (Math|Vector|Normalize (+ (* Direction%20Offset%20Right (Math|Vector|GetRightVector BaseRotation)) (Utilities|Operators|Add (Math|Vector|GetForwardVector BaseRotation) (Math|Vector|MakeVector 0.0 0.0 Direction%20Offset%20Up)))))))

(fn AdjustSpawnRotation
  (params (InputPin :type Rotator) (AngleDeg :type Float%20%28single-precision%29))
  (return (Math|Rotator|MakeRotfromYX (Math|Vector|RotateVectorAroundAxis (Math|Vector|GetRightVector InputPin) AngleDeg (Math|Vector|GetForwardVector InputPin)) (Math|Vector|GetForwardVector InputPin))))

(fn ThrowArcProjectile
  (params (Class :type Prj%20Annoyed%20Mop%20Ark%20Parent%20Class%20Reference) (Scale :type Vector) (SpawnLocation :type Vector) (Player :type Actor%20Object%20Reference) (Rotation :type Rotator) (TiltAngleDeg :type Float%20%28single-precision%29) (InitialSpeed :type Float%20%28single-precision%29 :default 8e+03))
  (bind _returnvalue (Projectile|SpawnProjectile :ProjectileClass Class :RequestedBy Player :SpawnLocation SpawnLocation :SpawnRotation (Private|AdjustSpawnRotation Rotation TiltAngleDeg) :SpawnDirection Rotation :bUseDefaultPhysics false :InitialSpeed InitialSpeed :GravityScale 0.0 :ChargePercent 0.0))
  (return _returnvalue))
