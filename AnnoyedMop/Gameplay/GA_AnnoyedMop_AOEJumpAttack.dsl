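; Gameplay ability for the AnnoyedMop jump slam, launching the player into an area attack on landing.
; Activation exits rails, ziplines or open seats, then a delayed DoJump applies root-motion force shaped by input, sprint, slide and an overhead trace.
; Landing, or optionally water entry, runs server-side targeted damage, impact effects and cooldown, then ends.
; Underwater checks block activation and jump; knockback or a not-falling timer check ends it early.
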(event Ability|EventActivateAbility
  (bind _returnvalue (Ability|GetAvatarActorfromActorInfo))
  ;[1 1 1] On Activate Ability jump the player.
  (Utilities|FlowControl|Sequence
    (:then_0
      (CallFunction|ResetLanded)
      ;;[1 1 1] PlayerPawn is actually a var of a parent GA and normally would be set in InitAbility called by parent's ActivateAbility.
      ;; Since we're not calling parent's ActivateAbility until later, set the PlayerPawn variable ourself here
      (Utilities|IsValid _returnvalue
        (:Is%20Valid
          (bind _asfort_player_pawn (Utilities|Casting|CastToFortPlayerPawn _returnvalue))
          (Variables|Default|SetPlayerPawn _asfort_player_pawn))))
    (:then_1
      (Variables|Default|SetWasSprintingWhenAbilityStarted (Variables|/Script/FortniteGame.FortPlayerPawn|GetbIsTacticalSprinting (Variables|Default|GetPlayerPawn))))
    (:then_2
      (CallFunction|MaybeCancelOtherAbilitiesorMovement) ; Check if should eject from zipline, grind rail, etc 
      (ForceFeedback|SpawnForceFeedbackAttached "/AnnoyedMop/Gameplay/ForceFeedback/FF_JumpSlam_Jump_AnnoyedMop.FF_JumpSlam_Jump_AnnoyedMop" (Variables|/Script/Engine.Character|GetMesh (Variables|Default|GetPlayerPawn))))
    (:then_3)
    (:then_4
      (Utilities|Time|SetTimerbyEvent (delegate Custom|DoJump) 0.2)
      (Pawn|Character|ResetFallingHeight (Variables|Default|GetPlayerPawn)))))

(event Custom|DoJump
  (bind _playerpawn (Variables|Default|GetPlayerPawn))
  (bind _returnvalue (Transformation|GetActorLocation _playerpawn))
  ;[1 1 1] On Activate Ability jump the player.
  (if (not (CallFunction|ShouldBlockAbilityDuetoUnderwater))
    (bind _returnvalue_2 (Game|Player|AddNewCameraModifier (Variables|/Script/Engine.PlayerController|GetPlayerCameraManager (Fortnite|GetFortPlayerControllerfromActor (Variables|Default|GetPlayerPawn))) "/AnnoyedMop/Gameplay/CameraModifier_AOEJumpSlam_AnnoyedMop.CameraModifier_AOEJumpSlam_AnnoyedMop_C"))
    (Variables|Default|SetCameraMod _returnvalue_2)
    (CallFunction|BindonKnockedBack)
    (Variables|Default|SetInitialForwardVector (Transformation|GetActorForwardVector (Variables|Default|GetPlayerPawn)))
    (Variables|Default|SetInitialVelocityClamped (Math|Vector|VectorClampSizeMax (Transformation|GetVelocity (Variables|Default|GetPlayerPawn)) (select (>= (ScalableFloat|GetValueatLevel (Variables|Balance|GetMaxSpeedToMaintain)) 0.0) (ScalableFloat|GetValueatLevel (Variables|Balance|GetMaxSpeedToMaintain)) (Math|Vector|VectorLength (Transformation|GetVelocity (Variables|Default|GetPlayerPawn))))))
    (Ability|ExecuteGameplayCueWithParamsOnOwner (Variables|Default|GetGC_Jump) (Ability|GameplayCue|MakeGameplayCueParameters :NormalizedMagnitude 0.0 :RawMagnitude 0.0 :Location _returnvalue :Normal "0,0,0" :GameplayEffectLevel 1 :AbilityLevel 1 :bReplicateLocationWhenUsingMinimalRepProxy false))
    (bind (_outhit _returnvalue_3) (Collision|LineTraceForObjects _returnvalue (+ _returnvalue "0.000000,0.000000,400.000000") (Utilities|Array|MakeArray "ObjectTypeQuery1") false (Utilities|Array|MakeArray (Variables|Default|GetPlayerPawn)) (select (Variables|Default|GetDebugJump) "ForDuration" "None")))
    (Ability|Tasks|ApplyRootMotionConstantForce "None" (Math|Vector|Normalize (+ (select _returnvalue_3 "0.000000,0.000000,500.000000" "0.000000,0.000000,830.000000") (Math|Vector|SelectVector (* (Math|Vector|Normalize (Variables|/Script/Engine.MovementComponent|GetVelocity (Variables|/Script/Engine.Character|GetCharacterMovement (Variables|Default|GetPlayerPawn)))) (ScalableFloat|GetValueatLevel (Variables|Balance|GetJumpDistance_Normal))) (* (Math|Vector|Normalize (Pawn|Components|CharacterMovement|GetCurrentAcceleration (Variables|/Script/Engine.Character|GetCharacterMovement (Variables|Default|GetPlayerPawn)))) (select (> (Math|Vector|VectorLength (Pawn|Components|CharacterMovement|GetCurrentAcceleration (Variables|/Script/Engine.Character|GetCharacterMovement (Variables|Default|GetPlayerPawn)))) 1e+02) (select (Variables|Default|GetWasSprintingWhenAbilityStarted) (ScalableFloat|GetValueatLevel (Variables|Balance|GetJumpDistance_Sprinting)) (ScalableFloat|GetValueatLevel (Variables|Balance|GetJumpDistance_Normal))) 0.0)) (Variables|/Script/FortniteGame.FortPawn|GetbIsSliding (Variables|Default|GetPlayerPawn))))) (Math|Vector|VectorLength (+ (select _returnvalue_3 "0.000000,0.000000,500.000000" "0.000000,0.000000,830.000000") (Math|Vector|SelectVector (* (Math|Vector|Normalize (Variables|/Script/Engine.MovementComponent|GetVelocity (Variables|/Script/Engine.Character|GetCharacterMovement (Variables|Default|GetPlayerPawn)))) (ScalableFloat|GetValueatLevel (Variables|Balance|GetJumpDistance_Normal))) (* (Math|Vector|Normalize (Pawn|Components|CharacterMovement|GetCurrentAcceleration (Variables|/Script/Engine.Character|GetCharacterMovement (Variables|Default|GetPlayerPawn)))) (select (> (Math|Vector|VectorLength (Pawn|Components|CharacterMovement|GetCurrentAcceleration (Variables|/Script/Engine.Character|GetCharacterMovement (Variables|Default|GetPlayerPawn)))) 1e+02) (select (Variables|Default|GetWasSprintingWhenAbilityStarted) (ScalableFloat|GetValueatLevel (Variables|Balance|GetJumpDistance_Sprinting)) (ScalableFloat|GetValueatLevel (Variables|Balance|GetJumpDistance_Normal))) 0.0)) (Variables|/Script/FortniteGame.FortPawn|GetbIsSliding (Variables|Default|GetPlayerPawn))))) 0.2)
    (if _returnvalue_3
      (GameplayEffects|ApplyGameplayEffectToSelf :self (Ability|GetFortAbilitySystemComponent (Variables|Default|GetPlayerPawn)) :GameplayEffectClass "/AnnoyedMop/Gameplay/Effects/GE_AnnoyedMop_AOEJumpAttack_LowGrav.GE_AnnoyedMop_AOEJumpAttack_LowGrav_C" :Level 0.0)
      (else
        _))
    ;;[1 1 1] Parent:ActivateAbility
    (|CallParentFunction)
    (EventDispatchers|BindEventtoOnPawnLanded (Variables|Default|GetPlayerPawn) (EventDispatchers|CreateEvent :!event Custom|Landed))
    (bind _returnvalue_4 (Utilities|Time|SetTimerbyEvent (delegate Custom|CheckForInvalidState) 0.3 true))
    (Variables|Default|SetCheckInvalidStateTimer _returnvalue_4)
    (else
      ;;[1 0.8333 0.3051] Early exit due to being underwater.
      ;; This is also checked in CanActivateAbility, but check again here since there is a delay between activate and DoJump. Cancelling it early here looks better than letting it fail later due to invalid state since the force won't be enough for the player to exit the water.
      (Development|PrintString "Canceling DoJump due to underwater" (Variables|Default|GetDebugWater) (Variables|Default|GetDebugWater))
      (FailedAction|UnabletoPerformAction (Variables|Default|GetPlayerPawn) true (GameplayTags|MakeLiteralGameplayTagContainer "(GameplayTags=((TagName=\"Athena.Ability.Failed.CannotDoThatNow\")))"))
      (Ability|EndAbility))))

(event Custom|CheckForInvalidState
  ;[1 1 1] On Activate Ability jump the player.
  (if (and (not (GameplayTags|HasMatchingGameplayTag (Variables|Default|GetPlayerPawn) "(TagName=\"MovementMode.Falling\")")) (not (Variables|Default|GetHasLanded)))
    (Ability|EndAbility)))

(event Custom|Landed
  (params (Hit :type Hit%20Result%20Structure%20%28by%20ref%29))
  ;[1 1 1] On Landed
  (Utilities|FlowControl|DoOnce false
    (:Completed
      (Utilities|FlowControl|Sequence
        (:then_0
          ;;[1 0.6081 0.08314] Deal Damage
          (Ability|HasAuthority
            (:True
              (CallFunction|DealDamage))))
        (:then_1
          (Variables|Default|SetHasLanded true)
          (Ability|ExecuteGameplayCueWithParamsOnOwner (Variables|Default|GetGC_Impact) (Ability|GameplayCue|MakeGameplayCueParameters :NormalizedMagnitude 0.0 :RawMagnitude 0.0 :Location (Transformation|GetActorLocation (Variables|Default|GetPlayerPawn)) :Normal "0,0,0" :GameplayEffectLevel 1 :AbilityLevel 1 :bReplicateLocationWhenUsingMinimalRepProxy false))
          (Class|CameraModifierAOEJumpSlamAnnoyedMop|OnLanded (Variables|Default|GetCameraMod))
          (Components|Movement|StopMovementImmediately (Variables|/Script/Engine.Character|GetCharacterMovement (Variables|Default|GetPlayerPawn)))
          (bind _returnvalue (Ability|ApplyGameplayEffectToOwner (Variables|Default|GetGE_MoveSpeedOnLanded)))
          (Variables|Default|SetHandle_GE_MoveSpeedOnLanded _returnvalue)
          (bind _asynctaskproxy (Ability|Tasks|PlayMontageAndWaitNotify "None" "/AnnoyedMop/Assets/Animations/TEMP/AnnoyedMop_TEMP_GroundAOE_Slam_RadialEffects.AnnoyedMop_TEMP_GroundAOE_Slam_RadialEffects")
            (:then
              (Variables|Default|SetAsyncTask_Montage _asynctaskproxy)
              (if (Variables|Default|GetShouldApplyCooldown)
                (GameplayEffects|ApplyGameplayEffectToSelf :self (Ability|GetFortAbilitySystemComponent (Variables|Default|GetPlayerPawn)) :GameplayEffectClass "/AnnoyedMop/Gameplay/Effects/GE_AOEJumpSlam_Cooldown_AnnoyedMop.GE_AOEJumpSlam_Cooldown_AnnoyedMop_C" :Level 0.0)))
            (:OnCompleted
              _)
            (:OnBlendOut
              _)
            (:OnInterrupted
              _)
            (:OnCancelled
              _)) ; Currently using built in "Cast Montage" instead
          (Ability|EndAbility))))))

(event Custom|ResetLanded
  ;[1 1 1] On Landed
  (Variables|Default|SetHasLanded false)
  (Utilities|FlowControl|DoOnce false
    (:Completed
      (Utilities|FlowControl|Sequence
        (:then_0
          ;;[1 0.6081 0.08314] Deal Damage
          (Ability|HasAuthority
            (:True
              (CallFunction|DealDamage))))
        (:then_1
          (Variables|Default|SetHasLanded true)
          (Ability|ExecuteGameplayCueWithParamsOnOwner (Variables|Default|GetGC_Impact) (Ability|GameplayCue|MakeGameplayCueParameters :NormalizedMagnitude 0.0 :RawMagnitude 0.0 :Location (Transformation|GetActorLocation (Variables|Default|GetPlayerPawn)) :Normal "0,0,0" :GameplayEffectLevel 1 :AbilityLevel 1 :bReplicateLocationWhenUsingMinimalRepProxy false))
          (Class|CameraModifierAOEJumpSlamAnnoyedMop|OnLanded (Variables|Default|GetCameraMod))
          (Components|Movement|StopMovementImmediately (Variables|/Script/Engine.Character|GetCharacterMovement (Variables|Default|GetPlayerPawn)))
          (bind _returnvalue (Ability|ApplyGameplayEffectToOwner (Variables|Default|GetGE_MoveSpeedOnLanded)))
          (Variables|Default|SetHandle_GE_MoveSpeedOnLanded _returnvalue)
          (bind _asynctaskproxy (Ability|Tasks|PlayMontageAndWaitNotify "None" "/AnnoyedMop/Assets/Animations/TEMP/AnnoyedMop_TEMP_GroundAOE_Slam_RadialEffects.AnnoyedMop_TEMP_GroundAOE_Slam_RadialEffects")
            (:then
              (Variables|Default|SetAsyncTask_Montage _asynctaskproxy)
              (if (Variables|Default|GetShouldApplyCooldown)
                (GameplayEffects|ApplyGameplayEffectToSelf :self (Ability|GetFortAbilitySystemComponent (Variables|Default|GetPlayerPawn)) :GameplayEffectClass "/AnnoyedMop/Gameplay/Effects/GE_AOEJumpSlam_Cooldown_AnnoyedMop.GE_AOEJumpSlam_Cooldown_AnnoyedMop_C" :Level 0.0)))
            (:OnCompleted
              _)
            (:OnBlendOut
              _)
            (:OnInterrupted
              _)
            (:OnCancelled
              _)) ; Currently using built in "Cast Montage" instead
          (Ability|EndAbility))))))

(event Ability|EventOnEndAbility
  (params (bWasCancelled :type Boolean))
  ;[1 1 1] On End Ability - Unbind Land Events
  (|CallParentFunction bWasCancelled)
  (GameplayTasks|EndTask (Variables|Default|GetAsyncTask_Montage))
  (Class|CameraModifierAOEJumpSlamAnnoyedMop|DisableandRemove (Variables|Default|GetCameraMod))
  (Utilities|Time|ClearandInvalidateTimerbyHandle (Variables|Default|GetCheckInvalidStateTimer))
  (EventDispatchers|UnbindEventfromOnPawnLanded (Variables|Default|GetPlayerPawn) (EventDispatchers|CreateEvent :!event Custom|Landed))
  (Ability|RemoveGameplayEffectFromOwnerWithHandle (Variables|Default|GetHandle_GE_MoveSpeedOnLanded))
  (GameplayTasks|EndTask (Variables|Default|GetTask_OnKnockedBack))
  ;;[1 1 1] Removes tag the blocks grinding so players can grind again
  (Ability|GameplayEffect|RemoveLooseGameplayTags (Variables|Default|GetPlayerPawn) (Variables|Default|GetBlockGrindRailsTag)))

(event EventCastMontageCancelledOrInterrupted)

(event Custom|DealDamage
  ;[1 0.6964 0.1334] DealDamage (server only)
  (Ability|HasAuthority
    (:True
      (bind (_asynctaskproxy _targetdata) (Ability|Tasks|DoTargetSelection "AnnoyedMopJumpAttackDamage" (Variables|Default|GetApplicationTag_SlamDamage) true)
        (:Targeted
          (Ability|Tasks|ApplyGameplayEffectToTargetDataAsync _targetdata (Variables|Default|GetGE_SlamDamage) 1 1 (Variables|Default|GetMaxTargetsToDamagePerTick)) ; Apply GE async
        )) ; Find targets
    )))

(event EventPlayerEnterWater
  (params (FortPlayerPawn :type Fort%20Player%20Pawn%20Object%20Reference))
  (bind _debugwater (Variables|Default|GetDebugWater))
  ;[1 1 1] PlayerEnterWater
  (if (Variables|PowerWater|GetEndOnWaterEnter)
    (|CallParentFunction FortPlayerPawn) ; parent handles ending ability on entering water
    (else
      (Ability|HasAuthority
        (:True
          _)
        (:False
          (if (ScalableFloat|GetValueasBool (Variables|Default|GetbClientCheckDirectionOnEnterWater))
            ;;[1 1 1 1 14] The client can get "false" enter water events when exiting the water due to net correction (client exits water, server says no not yet and corrects client back into water, etc). To handle this, on the client, only call Landed if our z velocity is negative (we're falling down into the water). This fixes FX playing early on the client sometimes when using the ability in water.
            ;; Note: May also get multiple call landed events when actually entering the water for the same reason, but Landed is gated.
            (bind _charactermovement (Variables|/Script/Engine.Character|GetCharacterMovement FortPlayerPawn)
              (:then
                (if (< (.z (Variables|/Script/Engine.MovementComponent|GetVelocity _charactermovement)) 0)
                  (Development|PrintString "Enter water with negative Z on client - Call Landed" _debugwater _debugwater)
                  (else
                    (Development|PrintString "Enter water with postive Z on client - Ignoring" _debugwater _debugwater))))
              (:else))
            (else
              _))
          (CallFunction|Landed) ; Otherwise treat it like we landed
        )))))

(event Custom|BindOnKnockedBack
  ;[1 1 1] On Knocked Back
  (bind _asynctaskproxy (Ability|Tasks|WaitGameplayTagAdd (Variables|Default|GetTag_KnockedBack))
    (:then
      (Variables|Default|SetTask_OnKnockedBack _asynctaskproxy))
    (:Added
      (Ability|EndAbility))))

(fn CanActivateAbility
  (params (ActorInfo :type Gameplay%20Ability%20Actor%20Info%20Structure) (Handle :type Gameplay%20Ability%20Spec%20Handle%20Structure))
  (bind _asfort_player_pawn (Utilities|Casting|CastToFortPlayerPawn (Ability|GetAvatarActorfromActorInfo))
    (:then
      (bind _returnvalue (GameplayTags|HasMatchingGameplayTag(Message) _asfort_player_pawn "(TagName=\"Granted.Cooldown.AnnoyedMop\")"))
      (if _returnvalue
        ;[1 1 1] Failed due to cooldown
        (FailedAction|UnabletoPerformAction _asfort_player_pawn true (GameplayTags|MakeGameplayTagContainerfromTag "(TagName=\"Athena.Ability.Failed.Cooldown\")")) ; show reason for failing on screen if one is given
        (return "(GameplayTags=)" false)
        (elif (CallFunction|ShouldBlockAbilityDuetoUnderwater)
          ;[1 1 1] Failed due to underwater.
          ; (Block activating underwater as otherwise the force won't be enough to exit the water and the ability will fail out later on which doesn't look great.)
          (FailedAction|UnabletoPerformAction _asfort_player_pawn true (GameplayTags|MakeGameplayTagContainerfromTag "(TagName=\"Athena.Ability.Failed.CannotDoThatNow\")")) ; show reason for failing on screen if one is given
          (Development|PrintString "Ability failed due to underwater" (Variables|Default|GetDebugWater) (Variables|Default|GetDebugWater))
          (return "(GameplayTags=)" false))))
    (:CastFailed
      _))
  (bind (_relevanttags _returnvalue_2) (|CallParentFunction ActorInfo Handle))
  (return _relevanttags _returnvalue_2))

(fn MaybeCancelOtherAbilitiesOrMovement
  (bind _playerpawn (Variables|Default|GetPlayerPawn))
  (bind _returnvalue (Vehicle|GetVehicle _playerpawn))
  (bind _playerpawn_2 (Variables|Default|GetPlayerPawn))
  ;[1 0.5543 0.2413] Checks if player should be ejected from their current movement mode - extended from witches broom
  ;;[1 1 1] Grind Rails
  (if (GameplayTags|HasMatchingGameplayTag (Variables|Default|GetPlayerPawn) "(TagName=\"MovementMode.Grinding\")")
    (Ability|GameplayEffect|AddLooseGameplayTags _playerpawn_2 (Variables|Default|GetBlockGrindRailsTag))
    (Pawn|Components|CharacterMovement|SetMovementMode (Variables|/Script/Engine.Character|GetCharacterMovement _playerpawn_2) "MOVE_Walking" 0)
    (else
      ;;[1 1 1] Ziplining
      (if (GameplayTags|HasAnyMatchingGameplayTags (Variables|Default|GetPlayerPawn) (GameplayTags|MakeLiteralGameplayTagContainer "(GameplayTags=((TagName=\"MovementMode.ZipliningOnSpline\"),(TagName=\"MovementMode.Ziplining\")))"))
        (Zipline|EndZiplining (Variables|Default|GetPlayerPawn) false false)
        (elif (GameplayTags|HasMatchingGameplayTag (Variables|Default|GetPlayerPawn) "(TagName=\"Gameplay.Status.InVehicle.Seat.Open\")")
          ;;[1 1 1] Open seated vehicle
          (Utilities|IsValid _returnvalue
            (:Is%20Valid
              (bind _returnvalue_2 (Vehicle|EjectPlayerPawn(Message) _returnvalue _playerpawn))
              (if (not _returnvalue_2)
                ;;;[1 0 0] If we somehow fail to eject, end the ability
                (Ability|EndAbility)))))))))

(fn ShouldBlockAbilityDueToUnderwater
  (return (and (ScalableFloat|GetValueasBool (Variables|Default|GetbShouldBlockAbilityUnderwater)) (> (WaterBody|GetWaterImmersionDepth (CallFunction|GetAvatarasFortPlayerPawn)) (* (Components|Capsule|GetScaledCapsuleHalfHeight (Variables|/Script/Engine.Character|GetCapsuleComponent (CallFunction|GetAvatarasFortPlayerPawn))) 2.0)))))
