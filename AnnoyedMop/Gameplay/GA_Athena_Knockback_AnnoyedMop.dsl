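; Knockback gameplay ability for AnnoyedMop that specializes a parent knockback ability.
; GetKnockbackOrigin places the origin at the target's location minus the origin actor's forward vector, so the knock direction follows that actor's facing rather than its position.
; GetInitialKnockMag prints a leftover debug string and returns the parent's result.
; The activate and end events add no logic of their own.
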(event Ability|EventActivateAbility
  (|CallParentFunction) ; This node is disabled and will not be called.\nDrag off pins to build functionality.
)

(event Ability|EventOnEndAbility
  (params (bWasCancelled :type Boolean)))

(fn GetInitialKnockMag
  (params (Direction :type Vector :default "0, 0, 0") (OverrideXYMag :type Boolean) (XYMAgOverride :type Float%20%28double-precision%29))
  (Development|PrintString "bruh")
  (return (|CallParentFunction Direction)))

(fn GetKnockbackOrigin
  (params (OriginActor :type Actor%20Object%20Reference))
  (return (- (Transformation|GetActorLocation (Variables|KnockbackParent|GetTarget)) (Transformation|GetActorForwardVector OriginActor))))
