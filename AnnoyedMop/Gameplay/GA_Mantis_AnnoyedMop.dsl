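; Mantis gameplay ability for the AnnoyedMop weapon's combo and air attacks, which throw tilted arc projectiles.
; Technique-start events route to timed attack functions that drive weapon events, camera modifier, gameplay cues and force feedback.
; ThrowProjectile spawns server-side after AdjustAim sphere-traces to snap aim onto nearby pawns within a dot threshold.
; It also picks montages by movement direction, gates use while swimming and makes projectiles ignore the rider's motorcycle.
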(event Mantis|EventOnMantisAblilityTechniqueStarted
  (params (OwningTags :type Gameplay%20Tag%20Container%20Structure) (SequenceName :type Name :default "None") (StartsSequence :type Boolean) (EndsSequence :type Boolean))
  ;[1 1 1] On Mantis Ability Technique Started
  (switch name SequenceName
    (:Attack1
      (Variables|Default|SetbForceUnCrouch true))
    (:Attack2
      (Variables|Default|SetbForceUnCrouch true)
      (CallFunction|Attack2))
    (:Attack3
      (Variables|Default|SetbForceUnCrouch true)
      (CallFunction|Attack3))
    (:Attack4
      (Variables|Default|SetbForceUnCrouch true)
      (CallFunction|Attack4))
    (:AirAttack
      (Variables|Default|SetbForceUnCrouch true)
      (CallFunction|AirAttack))
    (:Attack1_UpperOnly
      (Variables|Default|SetbForceUnCrouch false))
    (:Attack2_UpperOnly
      (Variables|Default|SetbForceUnCrouch false)))
  (CallFunction|Attack1))

(event Ability|EventActivateAbility
  (bind _returnvalue (Ability|GetActivatingPawn))
  (bind _returnvalue_2 (Ability|GetActivatingPawn))
  ;[0.6812 1 0.6971] On Activate Ability
  (|CallParentFunction)
  (Utilities|FlowControl|Sequence
    (:then_0
      (Variables|AimAdjustment|SetbAdjustAimToTargetedPlayer (ScalableFloat|GetValueasBool (Variables|AimAdjustment|GetbAdjustAimToTargetedPlayer_HF)))
      (Variables|AimAdjustment|SetAdjustAimToTargetedPlayer_MinDot (ScalableFloat|GetValueatLevel (Variables|AimAdjustment|GetAdjustAimToTargetedPlayer_MinDot_HF)))
      (Variables|AimAdjustment|SetAdjustAimToTargetedPlayer_TraceRadius (ScalableFloat|GetValueatLevel (Variables|AimAdjustment|GetAdjustAimToTargetedPlayer_TraceRadius_HF)))
      (Variables|AimAdjustment|SetAdjustAimToTargetedPlayer_TraceDist (ScalableFloat|GetValueatLevel (Variables|AimAdjustment|GetAdjustAimToTargetedPlayer_TraceDist_HF))))
    (:then_1
      (bind _returnvalue_3 (Game|Player|AddNewCameraModifier (Variables|/Script/Engine.PlayerController|GetPlayerCameraManager (Fortnite|GetFortPlayerControllerfromActor _returnvalue)) "/AnnoyedMop/Gameplay/CameraModifier_AnnoyedMop.CameraModifier_AnnoyedMop_C"))
      (bind _output_get (Variables|Default|SetCamera%20Mod _returnvalue_3))
      (Class|CameraModifierAnnoyedMop|TriggerComboStart _output_get)
      (bind _asb_weapon_annoyed_mop (Utilities|Casting|CastToB_Weapon_AnnoyedMop (Variables|/Script/FortniteGame.FortPawn|GetCurrentWeapon _returnvalue)))
      (Variables|Default|SetWeapon _asb_weapon_annoyed_mop))
    (:then_2
      ;;[1 1 1] If we're on a motorcycle, ignore the motorcycle for projectiles.
      ;; (Otherwise we'll hit it when we fire. Animal Riding does this)
      (Ability|HasAuthority
        (:True
          (if (GameplayTags|HasMatchingGameplayTag _returnvalue_2 "(TagName=\"Gameplay.Action.Player.InVehicle.Motorcycle\")")
            (bind _asfort_player_pawn (Utilities|Casting|CastToFortPlayerPawn _returnvalue_2))
            (Utilities|IsValid (Vehicle|GetVehicle _asfort_player_pawn)
              (:Is%20Valid
                (Class|FortPawn|AddProjectileIgnoreActor _asfort_player_pawn (Vehicle|GetVehicle _asfort_player_pawn))
                (Variables|Default|SetProjectileIgnoredMotorcycle (Vehicle|GetVehicle _asfort_player_pawn)) ; store this so we can remove it later
              ))))))))

(event Custom|Attack1
  ;[1 1 1] Attack 1
  (Default|CallTriggeredComboAttack (Variables|Default|GetWeapon) 1)
  (CallFunction|ThrowProjectile 1e+01 "/AnnoyedMop/Gameplay/ProjectileChildren/Prj_AnnoyedMop_Ark_Basic.Prj_AnnoyedMop_Ark_Basic_C")
  (Class|CameraModifierAnnoyedMop|TriggerPullBack (Variables|Default|GetCamera%20Mod) 0.4 0.4 0.0)
  (Utilities|FlowControl|Delay 0.001)
  (CallFunction|ExecuteGameplayCue (Variables|Default|GetBasicAttackTag))
  (ForceFeedback|SpawnForceFeedbackAttached "/AnnoyedMop/Gameplay/ForceFeedback/FF_Combo1_AnnoyedMop.FF_Combo1_AnnoyedMop" (Variables|/Script/FortniteGame.FortWeapon|GetWeaponMesh (Variables|Default|GetWeapon))))

(event Custom|Attack2
  ;[1 1 1] Attack 2
  (Default|CallTriggeredComboAttack (Variables|Default|GetWeapon) 2)
  (CallFunction|ThrowProjectile -1e+01 "/AnnoyedMop/Gameplay/ProjectileChildren/Prj_AnnoyedMop_Ark_Basic.Prj_AnnoyedMop_Ark_Basic_C")
  (Class|CameraModifierAnnoyedMop|TriggerPullBack (Variables|Default|GetCamera%20Mod) 0.3 0.4 0.05)
  (Utilities|FlowControl|Delay 0.001)
  (CallFunction|ExecuteGameplayCue (Variables|Default|GetBasicAttackTag))
  (ForceFeedback|SpawnForceFeedbackAttached "/AnnoyedMop/Gameplay/ForceFeedback/FF_Combo1_AnnoyedMop.FF_Combo1_AnnoyedMop" (Variables|/Script/FortniteGame.FortWeapon|GetWeaponMesh (Variables|Default|GetWeapon))))

(event Ability|EventOnEndAbility
  (params (bWasCancelled :type Boolean))
  (bind _returnvalue (Ability|GetActivatingPawn))
  (bind _projectileignoredmotorcycle (Variables|Default|GetProjectileIgnoredMotorcycle))
  ;[1 0.8522 0.6062] EndAbility
  (|CallParentFunction bWasCancelled)
  (Utilities|FlowControl|Sequence
    (:then_0
      (Class|CameraModifierAnnoyedMop|DisableandRemove (Variables|Default|GetCamera%20Mod)))
    (:then_1
      ;;[1 1 1] Remove ProjectileIgnoredMotorcycle from pawn's ProjectileIgnoredActors
      (Ability|HasAuthority
        (:True
          (Utilities|IsValid _returnvalue
            (:Is%20Valid
              (Utilities|IsValid _projectileignoredmotorcycle
                (:Is%20Valid
                  (Class|FortPawn|RemoveProjectileIgnoreActor _returnvalue _projectileignoredmotorcycle)
                  (Variables|Default|SetProjectileIgnoredMotorcycle 0) ; clear this out so we don't hold onto an old\n motorcycle reference on future activations
                )))))))))

(event Custom|Attack3
  ;[1 1 1] Attack 3
  (Default|CallTriggeredComboAttack (Variables|Default|GetWeapon) 3)
  (Class|CameraModifierAnnoyedMop|TriggerPullBack (Variables|Default|GetCamera%20Mod) 2.0 0.45 0.0)
  (Utilities|FlowControl|Delay 0.1)
  (CallFunction|ThrowProjectile -15.0 "/AnnoyedMop/Gameplay/ProjectileChildren/Prj_AnnoyedMop_Ark_Large.Prj_AnnoyedMop_Ark_Large_C")
  (CallFunction|ExecuteGameplayCue (Variables|Default|GetBasicKickAttackTag))
  (ForceFeedback|SpawnForceFeedbackAttached "/AnnoyedMop/Gameplay/ForceFeedback/FF_Combo2_AnnoyedMop.FF_Combo2_AnnoyedMop" (Variables|/Script/FortniteGame.FortWeapon|GetWeaponMesh (Variables|Default|GetWeapon)))
  (Utilities|FlowControl|Delay 0.3)
  (CallFunction|ThrowProjectile 15.0 "/AnnoyedMop/Gameplay/ProjectileChildren/Prj_AnnoyedMop_Ark_Large.Prj_AnnoyedMop_Ark_Large_C")
  (CallFunction|ExecuteGameplayCue (Variables|Default|GetBasicKickAttackTag))
  (ForceFeedback|SpawnForceFeedbackAttached "/AnnoyedMop/Gameplay/ForceFeedback/FF_Combo3_AnnoyedMop.FF_Combo3_AnnoyedMop" (Variables|/Script/FortniteGame.FortWeapon|GetWeaponMesh (Variables|Default|GetWeapon))))

(event Custom|Attack4
  ;[1 1 1] Attack 4
  (Class|CameraModifierAnnoyedMop|DisableandRemove (Variables|Default|GetCamera%20Mod))
  (Default|CallTriggeredComboAttack (Variables|Default|GetWeapon) 4)
  (Utilities|FlowControl|Delay 0.44)
  (bind (_spawnedprojectiles _projectileaimrotation) (CallFunction|ThrowProjectile 0.0 "/AnnoyedMop/Gameplay/ProjectileChildren/Prj_AnnoyedMop_Ark_Massive.Prj_AnnoyedMop_Ark_Massive_C"))
  (CallFunction|ExecuteGameplayCue (Variables|Default|GetLargeAttackTag) (Math|Rotator|GetRotationXVector _projectileaimrotation))
  (ForceFeedback|SpawnForceFeedbackAttached "/AnnoyedMop/Gameplay/ForceFeedback/FF_Combo4_AnnoyedMop.FF_Combo4_AnnoyedMop" (Variables|/Script/FortniteGame.FortWeapon|GetWeaponMesh (Variables|Default|GetWeapon))))

(event Custom|Air%20Attack
  ;[1 1 1] Air Attack
  (Default|CallTriggeredinAirAttack (Variables|Default|GetWeapon))
  (CallFunction|ThrowProjectile 4.0 "/AnnoyedMop/Gameplay/ProjectileChildren/Prj_AnnoyedMop_Ark_Large.Prj_AnnoyedMop_Ark_Large_C")
  (Utilities|FlowControl|Delay 0.001)
  (CallFunction|ExecuteGameplayCue (Variables|Default|GetAirAttackTag)))

(fn AdjustAim
  (params (ProjectileSpawnLocation :type Vector))
  (bind _originalspawnrotation (LocalVariables|GetOriginalSpawnRotation))
  (bind _returnvalue (Math|Vector|GetForwardVector _originalspawnrotation))
  (LocalVariables|SetOriginalSpawnRotation (Default|AdjustArcProjectileSpawnDirection (Utilities|GetCameraRotation) (LocalVariables|GetDirOffsetRight) (LocalVariables|GetDirOffsetUp)))
  (Utilities|FlowControl|Sequence
    (:then_0
      ;[0.149 0.3059 0.8941] Adjust aim to targeted player (if enabled)
      ; Because the projectile fires from around midheight of the player, it doesn't always hit what is under the reticle
      ; when aiming at something very close (eg dbno player at feet). To counteract this, search for and add a bit of
      ; auto aim to nearby players.
      (if (Variables|AimAdjustment|GetbAdjustAimToTargetedPlayer)
        (bind _weapon (Variables|Default|GetWeapon)
          (:then
            (FortWeapon|UpdateDamageStartLocation _weapon) ; Update it before getting as it may not have been set yet on the server
            (bind _output_get (LocalVariables|SetTraceStart (.ReturnValue (FortWeapon|GetDamageStartLocation _weapon -1))))
            (bind (_outhit _returnvalue_2) (Collision|SphereTraceForObjects _output_get (+ _output_get (* (FortWeapon|GetDamageStartLocation _weapon -1) (Variables|AimAdjustment|GetAdjustAimToTargetedPlayer_TraceDist))) (Variables|AimAdjustment|GetAdjustAimToTargetedPlayer_TraceRadius) (Utilities|Array|MakeArray "ObjectTypeQuery3") false (Utilities|Array|MakeArray (Ability|GetActivatingPawn)) (select (LocalVariables|GetbDebugDrawAimTrace) "ForDuration" "None") true "(R=0.000000,G=0.477117,B=1.000000,A=1.000000)" "(R=0.237847,G=0.005192,B=0.036048,A=1.000000)" 2.0)) ; Since the trace radius should be fairly small\nand this is mostly to handle issues with nearby pawns,\njust single trace and use the nearest hit.
            (if _returnvalue_2
              (bind _output_get_2 (LocalVariables|SetRotationToHitPawn (Math|Rotator|MakeRotfromXY (- (.ImpactPoint (Collision|BreakHitResult _outhit)) (Variables|GetProjectileSpawnLocation)) (Math|Vector|GetRightVector (LocalVariables|GetOriginalSpawnRotation)))))
              (LocalVariables|SetValidHit (> (Math|Vector|DotProduct (Math|Vector|GetForwardVector _output_get_2) _returnvalue) (Variables|AimAdjustment|GetAdjustAimToTargetedPlayer_MinDot)))
              (if (LocalVariables|GetbDebugDrawAimAdjustment)
                ;;[0.2066 0.2066 0.2066] Debug
                (Development|PrintString (Utilities|String|Append "Dot " (Utilities|String|ToString(Float) (Math|Vector|DotProduct (Math|Vector|GetForwardVector _output_get_2) _returnvalue))))
                (Rendering|Debug|DrawDebugSphere (.ImpactPoint (Collision|BreakHitResult _outhit)) 1e+01 12 "(R=1.000000,G=0.015551,B=0.000000,A=1.000000)" 2.0) ; impact point on hit pawn
                (Rendering|Debug|DrawDebugLine (Variables|GetProjectileSpawnLocation) (+ (Variables|GetProjectileSpawnLocation) (* (Math|Vector|GetForwardVector (LocalVariables|GetOriginalSpawnRotation)) (Variables|AimAdjustment|GetAdjustAimToTargetedPlayer_TraceDist))) "(R=0.000000,G=0.098039,B=1.000000,A=1.000000)" 2.0 1.0) ; Original rotation forward vector
                (Rendering|Debug|DrawDebugLine (Variables|GetProjectileSpawnLocation) (+ (Variables|GetProjectileSpawnLocation) (* (Math|Vector|GetForwardVector (LocalVariables|GetRotationToHitPawn)) (Variables|AimAdjustment|GetAdjustAimToTargetedPlayer_TraceDist))) (Math|Color|SelectColor "(R=0.066546,G=1.000000,B=0.000000,A=1.000000)" "(R=0.616320,G=0.301948,B=0.027567,A=1.000000)" (LocalVariables|GetValidHit)) 2.0 1.0) ; Potential new rotation foward  vector
              )
              (if (LocalVariables|GetValidHit)
                ;;[1 0.6745 0.2784] Return new rotation
                (return true (LocalVariables|GetRotationToHitPawn)))))
          (:else))))
    (:then_1
      ;[1 0.6745 0.2784] Return original rotation
      (return false (LocalVariables|GetOriginalSpawnRotation)))))

(fn ExecuteGameplayCue
  (params (GameplayCueTag :type Gameplay%20Tag%20Structure) (ProjectileDirection :type Vector))
  (Ability|ExecuteGameplayCueWithParamsOnOwner GameplayCueTag (Ability|GameplayCue|MakeGameplayCueParameters :NormalizedMagnitude 0.0 :RawMagnitude 0.0 :Location (Transformation|GetActorLocation (Ability|GetActivatingPawn)) :Normal ProjectileDirection :Instigator (Ability|GetActivatingPawn) :GameplayEffectLevel 1 :AbilityLevel 1 :bReplicateLocationWhenUsingMinimalRepProxy false)))

(fn GetDynamicMontageNameForTechnique
  (params (TechniqueName :type Name :default "None"))
  (bind _returnvalue (Ability|GetActivatingPawn))
  (bind _charactermovement (Variables|/Script/Engine.Character|GetCharacterMovement _returnvalue))
  (bind _returnvalue_2 (Pawn|Components|PawnMovement|GetLastInputVector _charactermovement))
  (bind _returnvalue_3 (Transformation|GetActorTransform _returnvalue))
  (bind (_returnvalue_x _returnvalue_y) (Math|Vector|BreakVector (Math|Transform|InverseTransformDirection _returnvalue_3 _returnvalue_2)))
  (bind _returnvalue_4 (Math|Transform|InverseTransformDirection _returnvalue_3 _returnvalue_2))
  (bind _yaw (Math|Vector|GetYawPitchfromVector _returnvalue_4))
  (if (not (and (Math|Float|InRange(Float) _returnvalue_x -0.01 0.01) (Math|Float|InRange(Float) _returnvalue_y -0.01 0.01)))
    (if (Math|Float|InRange(Float) _yaw -145.0 -35.0)
      (LocalVariables|SetDynamicMontageName "MovingLeft")
      (elif (Math|Float|InRange(Float) _yaw 35.0 145.0)
        (LocalVariables|SetDynamicMontageName "MovingRight")
        (elif (Math|Float|InRange(Float) _yaw -35.0 35.0 false false)
          (LocalVariables|SetDynamicMontageName "MovingForward")
          (else
            (LocalVariables|SetDynamicMontageName "Stationary")))))
    (return (LocalVariables|GetDynamicMontageName))
    (else
      _)))

(fn CanActivateAbility
  (params (ActorInfo :type Gameplay%20Ability%20Actor%20Info%20Structure) (Handle :type Gameplay%20Ability%20Spec%20Handle%20Structure))
  (bind (_owneractor _avataractor) (Utilities|Struct|BreakGameplayAbilityActorInfo ActorInfo))
  (bind _asfort_pawn (Utilities|Casting|CastToFortPawn _avataractor))
  (if (WaterBody|IsSurfaceSwimming _asfort_pawn)
    (return 0 (ScalableFloat|GetValueasBool (Variables|ScalableFloats|GetSwimmingComboEnabled)))
    (else
      (return 0 true))))

(fn ThrowProjectile
  (params (TiltAngleDeg :type Float%20%28single-precision%29) (PrjClass :type Prj%20Annoyed%20Mop%20Ark%20Parent%20Class%20Reference))
  (bind _returnvalue (Ability|GetActivatingPawn))
  (bind _returnvalue_2 (Transformation|GetActorLocation _returnvalue))
  (bind _projectilespawnoffset (Variables|Default|GetProjectileSpawnOffset))
  (bind _returnvalue_3 (+ _returnvalue_2 _projectilespawnoffset))
  (Utilities|FlowControl|Sequence
    (:then_0
      (Ability|HasAuthority
        (:True
          (Utilities|IsValid _returnvalue
            (:Is%20Valid
              (bind (_foundplayer _modifiedrotator) (CallFunction|AdjustAim _returnvalue_3))
              (bind _spawnedprojectile (Class|FLAnnoyedMop|ThrowArcProjectile (Variables|GetPrjClass) "0, 0, 0" _returnvalue_3 _returnvalue _modifiedrotator (Variables|GetTiltAngleDeg) (ScalableFloat|GetValueatLevel (Variables|ScalableFloats|GetROW_Projectile_InitialSpeed))))
              (return (Utilities|Array|MakeArray _spawnedprojectile) _modifiedrotator))))) ; FortKismetLibrary::SpawnProjectile is a server only function, \nso don't bother doing any of this other work on the client
    )
    (:then_1
      (return 0 "0, 0, 0"))))
