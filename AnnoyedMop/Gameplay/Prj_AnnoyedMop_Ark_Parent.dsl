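; Parent Fortnite projectile for AnnoyedMop: a growing, multi-segment shot that applies a gameplay effect to what it hits.
; The server checks segment collision each tick, disables hit segments, affects each ability system once, and stops the projectile when none remain.
; A replicated segment-status array drives Niagara visuals and death FX for surviving segments, alongside in-air audio and impact cues.
; Water blocking can be delayed for swimming shooters, optionally using a reticle trace.
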(event EventBeginPlay
  (bind _ns_annoyedmop_prj_proto (Variables|Default|GetNS_AnnoyedMop_Prj_Proto))
  ;[0.2255 1 0.459] Begin Play
  (|CallParentFunction)
  (Utilities|FlowControl|Sequence
    (:then_0
      ;;[1 1 1] Set FX in code
      (FX|InitializeFXComponent _ns_annoyedmop_prj_proto)
      ;;[1 1 1] Make sure the FX is fully visible
      (Components|Activation|Activate _ns_annoyedmop_prj_proto true)
      (CallFunction|UpdateVisuals)
      (bind _asfort_player_pawn (Utilities|Casting|CastToFortPlayerPawn (Variables|Actor|GetInstigator)))
      (Variables|Default|SetWeapon (Variables|/Script/FortniteGame.FortPawn|GetCurrentWeapon _asfort_player_pawn)))
    (:then_1
      ;;[1 1 1 1 14] Set up FortProjectileSegmentedCollision's Segments.
      ;; This DESTROYS the SegmentInitializers
      (CallFunction|BPInitializeSegments))
    (:then_2
      (Networking|SwitchHasAuthority
        (:Authority
          (bind _shoulddelay (CallFunction|ShouldDelayBlockingWater))
          (if _shoulddelay
            ;;[1 1 1] Init Water Blocking
            (Utilities|Time|SetTimerbyEvent (delegate Custom|DelayedBlockHitsWithWater) (ScalableFloat|GetValueatLevel (Variables|WaterBlocking|GetBlockWaterDelay))) ; Delay blocking water for a small period of time to let the projectile travel out of the water
            (else
              (Variables|/Script/FortniteGame.FortProjectileSegmentedCollisionComponent|SetbBlockHitsWithWater true (Variables|Default|GetFortProjectileSegmentedCollision))))
          ;;[1 1 1] Check for Initial Collision
          (CallFunction|CollisionCheck))))
    (:then_3
      (Networking|CanExecuteCosmeticEvents
        (:True
          (if (not (Variables|Default|GetCollsion))
            (Sound|PlayinAirSound)))))))

(event Custom|DelayedBlockHitsWithWater
  ;[0.2255 1 0.459] Begin Play
  ;;[1 1 1] Init Water Blocking
  (Variables|/Script/FortniteGame.FortProjectileSegmentedCollisionComponent|SetbBlockHitsWithWater true (Variables|Default|GetFortProjectileSegmentedCollision))
  ;;[1 1 1] Check for Initial Collision
  (CallFunction|CollisionCheck))

(event EventTick
  (params (DeltaSeconds :type Float%20%28single-precision%29))
  ;[0.4549 0.9373 1] Tick
  (|CallParentFunction DeltaSeconds)
  (Size|UpdateSize (Utilities|Time|GetWorldDeltaSeconds) (Variables|Default|GetGrowthSpeed))
  (Networking|SwitchHasAuthority
    (:Authority
      (CallFunction|CollisionCheck))))

(event Custom|OnHitActor
  (params (Actor :type Actor%20Object%20Reference) (Hit :type Hit%20Result%20Structure))
  (bind _returnvalue (Ability|GetFortAbilitySystemComponent Actor))
  (bind _instigator (Variables|Actor|GetInstigator))
  (bind _returnvalue_2 (Ability|GetFortAbilitySystemComponent _instigator))
  (bind _alreadyhitascs (Variables|Default|GetAlreadyHitASCs))
  ;[1 0.7289 0.4762] OnHitActor
  (Utilities|IsValid _returnvalue
    (:Is%20Valid
      (if (not (Utilities|Array|ContainsItem _alreadyhitascs _returnvalue))
        (Utilities|Array|Add _alreadyhitascs _returnvalue) ; Store unique hit ASCs instead of Actors\n so we don't double hit the pawn and an attached actor \n(eg the pawn and the olympus wings)
        (bind _outeffectcontexthandle (CallFunction|ApplyHitResulttoEffectContext (GameplayEffects|MakeFortEffectContext _returnvalue_2 _instigator _instigator (Variables|Default|GetWeapon) 0) Hit))
        (GameplayEffects|ApplyGameplayEffectToTarget _returnvalue_2 (Variables|Default|GetGE%20to%20Apply%20On%20Hit%20Actor) _returnvalue 0.0 _outeffectcontexthandle)))))

(event Custom|CollisionCheck
  ;[0.4549 0.9373 1] Collision Check
  (Networking|SwitchHasAuthority
    (:Authority
      (bind _returnvalue (Class|FortProjectileSegmentedCollisionComponent|TickCollisionCheck (Variables|Default|GetFortProjectileSegmentedCollision)))
      (for _array_element _returnvalue
        (bind (_array_element_segmentindex _array_element_hitresult) (Utilities|Struct|BreakCollisionSegmentHit _array_element))
        (CallFunction|OnHitActor (.HitActor (Collision|BreakHitResult _array_element_hitresult)) _array_element_hitresult)
        (Class|FortProjectileSegmentedCollisionComponent|EnableCollisionSegment (Variables|Default|GetFortProjectileSegmentedCollision) _array_element_segmentindex false)
        (Utilities|Array|SetArrayElem (Variables|Default|GetSegmentEnabledStatusReplicated) _array_element_segmentindex) ; Update replicated array
        (Variables|Default|SetCollsion true))
      (if (Utilities|Array|IsNotEmpty _returnvalue)
        ;;[1 0.4428 0.1199] Maybe Start Death
        (bind (_output_segmentindex _output_hitresult) (Utilities|Struct|BreakCollisionSegmentHit (Utilities|Array|Get(aref) _returnvalue)))
        (CallFunction|StartDeathifAllSegmentsDisabled _output_hitresult)
        ;;[1 0.6285 0] If there was a hit this check, ForceNetUpdate to get the info to the client asap.
        ;; FORT-718860 - Helps make segment kills in niagara more reliable/on frame
        (Networking|ForceNetUpdate)
        (else
          (Variables|Default|SetCollsion false))))))

(event Custom|OnHitActorImpact
  ;[0.4549 0.9373 1] Collision Check
  (GameplayCues|ExecuteGameplayCue (Variables|Default|GetImpactTag)))

(event Custom|FXSpawnSegmentDeathFX
  (bind (_returnvalue_location _returnvalue_rotation _returnvalue_scale) (Math|Transform|BreakTransform (Transformation|GetWorldTransform (Variables|Default|GetNS_AnnoyedMop_Prj_Proto))))
  ;[1 0.3866 0 0.533] VFX - Spawns death fx for each segment, on projectile death
  (if (> (CallFunction|GetNumSegmentsEnabled) 0)
    (bind _returnvalue (Niagara|SpawnSystematLocation "/AnnoyedMop/Assets/VFX/proto/NS_AnnoyedMop_Prj_OnDeath.NS_AnnoyedMop_Prj_OnDeath" _returnvalue_location _returnvalue_rotation _returnvalue_scale true false))
    (VFX|FXSetSegmentParams (Variables|Default|GetNS_AnnoyedMop_Prj_Proto) _returnvalue)
    (Components|Activation|Activate _returnvalue false)))

(event Custom|StopInAirAudio
  (params (FadeOutDuration :type Float%20%28single-precision%29))
  ;[0.1124 0.4685 1] Audio
  ;
  (bind _inairloopaudio (Variables|Default|GetInAirLoopAudio)
    (:then
      (Audio|Components|Audio|FadeOut _inairloopaudio FadeOutDuration 0.0 "Linear" :!class /Script/Engine.AudioComponent)
      (if (not (Variables|Default|GetCollsion))
        (Utilities|FlowControl|DoOnce false
          (:Completed
            (Audio|PlaySoundatLocation (Variables|Default|GetStopSound) (Transformation|GetActorLocation)))) ; Prevent sound from playing on OnStop and EndPlay
      ))
    (:else)))

(event EventEndPlay
  (params (EndPlayReason :type EEndPlayReason%20Enum :default "Destroyed"))
  ;[1 0.7299 0.2392] EndPlay
  (|CallParentFunction EndPlayReason)
  (CallFunction|StopinAirAudio 0.25)
  (switch Utilities|FlowControl|Switch|SwitchonEEndPlayReason EndPlayReason
    (:Destroyed
      _)
    (:RemovedFromWorld
      _))
  (CallFunction|FXSpawnSegmentDeathFX))

(event EventOnStop
  (params (Hit :type Hit%20Result%20Structure%20%28by%20ref%29))
  ;[1 0.7299 0.2392] OnStop
  (|CallParentFunction Hit)
  (Actor|Tick|SetActorTickEnabled false) ; We disable tick on the server in StartDeathifAllSegmentsDisabled. \nStop it on the client here
  (CallFunction|StopinAirAudio 0.0))

(fn ApplyHitResultToEffectContext
  (params (InEffectContextHandle :type Gameplay%20Effect%20Context%20Handle%20Structure) (InHitResult :type Hit%20Result%20Structure))
  (bind _output_get (LocalVariables|SetEffectContextHandle InEffectContextHandle))
  (Ability|EffectContext|AddHitResult _output_get InHitResult)
  (return (LocalVariables|GetEffectContextHandle)))

(fn BP_InitializeSegments
  (bind _fortprojectilesegmentedcollision (Variables|Default|GetFortProjectileSegmentedCollision))
  (Collision|InitializeCollisionSegments (CallFunction|GetSegmentInitializers) _fortprojectilesegmentedcollision)
  (for _ (Variables|/Script/FortniteGame.FortProjectileSegmentedCollisionComponent|GetCollisionSegments _fortprojectilesegmentedcollision)
    (Utilities|Array|Add (Variables|Default|GetSegmentEnabledStatusReplicated) "true")))

(fn FXGetSegmentParams
  (return 0 (Math|Vector|MakeVector 1.0 1.0 1.0) 1))

(fn FXGetSegmentTransforms
  (return (Utilities|Array|MakeArray)))

(fn FXSetSegmentParams
  (params (SourceSystem :type Niagara%20Particle%20System%20Component%20Object%20Reference) (NiagaraSystem :type Niagara%20Particle%20System%20Component%20Object%20Reference))
  (bind _returnvalue (Utilities|Name|MakeLiteralName "SegmentPositions"))
  (bind _returnvalue_2 (Utilities|Name|MakeLiteralName "SegmentRotations"))
  (bind _niagarasystem (Variables|GetNiagaraSystem))
  (Utilities|FlowControl|Sequence
    (:then_0
      (bind _returnvalue_3 (Niagara|NiagaraGetQuaternionArray (Variables|GetSourceSystem) _returnvalue_2))
      (Niagara|NiagaraSetQuaternionArray (Variables|GetNiagaraSystem) _returnvalue_2 _returnvalue_3))
    (:then_1
      (bind _returnvalue_4 (Niagara|NiagaraGetVectorArray (Variables|GetSourceSystem) _returnvalue))
      (Niagara|NiagaraSetVectorArray (Variables|GetNiagaraSystem) _returnvalue _returnvalue_4))
    (:then_2
      (Niagara|NiagaraSetBoolArray (Variables|GetNiagaraSystem) "ActiveArray" (Variables|Default|GetSegmentEnabledStatusReplicated)))
    (:then_3
      (Niagara|SetNiagaraVariable(Int32) _niagarasystem "NumSegments" (VFX|FXGetSegmentParams))
      (Niagara|SetNiagaraVariable(Int32) _niagarasystem "SegmentWidth" 0))
    (:then_4
      (Niagara|SetNiagaraVariable(Vector3) (Variables|GetNiagaraSystem) "RootScale" "1.000000,1.000000,1.000000"))))

(fn GetNumSegmentsEnabled
  (for _array_element (Variables|Default|GetSegmentEnabledStatusReplicated)
    (if _array_element
      (Math|Integer|IncrementInt (LocalVariables|GetLocalNumSegmentsEnabled))))
  (return (LocalVariables|GetLocalNumSegmentsEnabled)))

(fn GetSegmentInitializers
  (return (Utilities|Array|MakeArray (Variables|Default|GetSegmentInitializer2_DestroyedAfterInit_DontAttach) (Variables|Default|GetSegmentInitializer_DestroyedAfterInit_DontAttach) (Variables|Default|GetSegmentInitializer1_DestroyedAfterInit_DontAttach))))

(fn OnRep_Collsion
  (if (Variables|Default|GetCollsion)
    (CallFunction|OnHitActorImpact)
    (CallFunction|StopinAirAudio 0.25)))

(fn OnRep_SegmentEnabledStatusReplicated
  (Networking|CanExecuteCosmeticEvents
    (:True
      (CallFunction|UpdateVisuals))))

(fn PlayInAirSound
  (Variables|Default|GetInAirLoopAudio
    (:then)
    (:else
      (bind _capsulecomponent (Variables|Collision|GetCapsuleComponent)
        (:then
          (bind _returnvalue (Audio|SpawnSoundAttached (Variables|Sound|GetInAirLoopSound) _capsulecomponent))
          (Variables|Default|SetInAirLoopAudio _returnvalue))
        (:else)))))

(fn ShouldDelayBlockingWater
  (if (> (ScalableFloat|GetValueatLevel (Variables|WaterBlocking|GetBlockWaterDelay)) 0)
    (bind _asfort_player_pawn (Utilities|Casting|CastToFortPlayerPawn (Variables|Actor|GetInstigator)))
    (if (WaterBody|IsSurfaceSwimming _asfort_player_pawn)
      (if (ScalableFloat|GetValueasBool (Variables|WaterBlocking|GetBlockWater_UseReticleAboveWaterTest))
        (bind _weapon (Variables|Default|GetWeapon)
          (:then
            ;[0.5191 0.7095 1] Test if the reticle is above the water line
            (FortWeapon|UpdateDamageStartLocation _weapon) ; No way to tell if the cached damage start location is up to date, so update
            (bind (_outhit _returnvalue) (Collision|LineTraceForObjects :Start (.ReturnValue (FortWeapon|GetDamageStartLocation _weapon -1)) :End (+ (.ReturnValue (FortWeapon|GetDamageStartLocation _weapon -1)) (* (FortWeapon|GetDamageStartLocation _weapon -1) (ScalableFloat|GetValueatLevel (Variables|WaterBlocking|GetBlockWater_ReticleAboveWaterTraceDist)))) :ObjectTypes (Utilities|Array|MakeArray "ObjectTypeQuery13") :bTraceComplex false :DrawDebugType (select (LocalVariables|GetbDebug) "ForDuration" "None") :bIgnoreSelf true :TraceColor "(R=1.000000,G=0.000000,B=0.000000,A=1.000000)" :TraceHitColor "(R=0.000000,G=1.000000,B=0.000000,A=1.000000)" :DrawTime 5.0)) ; Trace just for water to determine if the reticle is aimed above the water line
            ;[1 0.6235 0.2] If we hit water, the reticle is aimed below the water line. Block water immediately (no delay).
            ; Otherwise delay to give the projectile a chance to spawn and move out of the water
            (Development|PrintString (Utilities|String|SelectString "Reticle above water line, delaying blocking water" "Reticle below water line, block water immediately" (not _returnvalue)) (LocalVariables|GetbDebug) (LocalVariables|GetbDebug))
            (return (not _returnvalue)))
          (:else
            _))
        (else
          ;[1 0.6235 0.2] Delay for all swiming cases
          (Development|PrintString "UseReticleAboveWaterTest is false, delaying blocking water for all swimming cases" (LocalVariables|GetbDebug) (LocalVariables|GetbDebug))
          (return true)))
      (else
        _))
    ;[1 0.6235 0.2] No Delay
    (return false)
    (else
      (Development|PrintString "BlockWaterDelay is <=0, blocking water immediately" (LocalVariables|GetbDebug) (LocalVariables|GetbDebug)))))

(fn StartDeathIfAllSegmentsDisabled
  (params (HitThisTick :type Hit%20Result%20Structure))
  (bind (_bblockinghit _binitialoverlap _time _distance _location) (Collision|BreakHitResult HitThisTick))
  (Networking|SwitchHasAuthority
    (:Authority
      (LocalVariables|SetShouldStartDeath true)
      (bind _array_element (Utilities|Array|ForEachLoopwithBreak (Variables|Default|GetSegmentEnabledStatusReplicated))
        (:LoopBody
          (if _array_element
            (LocalVariables|SetShouldStartDeath false)))
        (:Completed
          _))
      (if (LocalVariables|GetShouldStartDeath)
        (Actor|SetLifeSpan (Variables|Delivery|GetLifespanAfterKill))
        (Game|Components|ProjectileMovement|StopSimulating (Variables|Movement|GetProjectileMovementComponent) (Collision|MakeHitResult false false 0.0 0.0 (+ (Transformation|GetActorLocation) (Math|Vector|ProjectVectorontoVector (- _location (Transformation|GetActorLocation)) (Transformation|GetActorForwardVector)))))
        (Actor|Tick|SetActorTickEnabled false)))))

(fn UpdateVisuals
  (Networking|CanExecuteCosmeticEvents
    (:True
      (for (_array_element _array_index) (Variables|Default|GetSegmentEnabledStatusReplicated)
        (Niagara|NiagaraSetBoolArrayValue (Variables|Default|GetNS_AnnoyedMop_Prj_Proto) "ActiveArray" _array_index _array_element)))))

(fn ConstructionScript)
